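# Material function that adds animated facial wrinkle normals and AO to input material attributes, with debug mask views.
# Eight expression regions from two stacked mask halves are weighted by left/right WM_ curve parameters, then max-combined into one alpha per half.
# These alphas drive angle-corrected normal blends and min-based AO blends from matching texture halves.
# The left/right split comes from pre-skinned local X or an optional texture; shading-path and static switches gate the effect.
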MaterialFunction /BRCosmetics/Characters/Player/Female/Medium/Bodies/F_MED_ElegantLilyAnkle/Materials/WM/MF_FacialExpressionBlend_2024
graph {nodes [
  n1 = Saturate(None=Add(A=NamedRerouteUsage(), B=NamedRerouteUsage()))
  n2 = NamedRerouteUsage()
  n3 = Constant()
  n4 = MaterialFunctionCall(Emissive=MaterialFunctionCall(X=NamedRerouteUsage(), Y=NamedRerouteUsage()) {materialFunction=&MakeFloat3}.Result) {materialFunction=&MF_FX_EmissiveAdjustments}
  n5 = NamedRerouteUsage()
  n6 = SetMaterialAttributes!(MaterialAttributes=NamedRerouteUsage(), Normal=NamedRerouteUsage(), AmbientOcclusion=NamedRerouteUsage()) {attributeSetTypes [#1, #2]}
  n7 = ScalarParameter() {defaultValue=1, group=FacialExpressionBlend, parameterName=FacialGlobalScale, sliderMax=1}
  n8 = StaticSwitchParameter(
   True=SetMaterialAttributes!(
    MaterialAttributes=NamedRerouteUsage()
    BaseColor=Constant()
    EmissiveColor=MaterialFunctionCall(Emissive=@n1) {materialFunction=&MF_FX_EmissiveAdjustments}.Result
   ) {attributeSetTypes [#3, #4]}
   False=StaticSwitchParameter(
    True=SetMaterialAttributes(MaterialAttributes=@n2, BaseColor=@n3, EmissiveColor=@n4.Result) {attributeSetTypes [#3, #4]}
    False=BlendMaterialAttributes(A=@n5, B=@n6, Alpha=@n7)
   ) {group=FacialExpressionBlend, parameterName=Debug Left/Right Mask, sortPriority=33}
  ) {group=FacialExpressionBlend, parameterName=Debug Expression Masks, sortPriority=33}
  n9 = NamedRerouteUsage()
  FunctionOutput(None=ShadingPathSwitch(
    Default=StaticSwitchParameter(True=@n8, False=@n9) {group=FacialExpressionBlend, parameterName=Use Facial Expression Blend, sortPriority=0}
    Mobile=NamedRerouteUsage()
   ))
  n10 = TextureSampleParameter2D() {
   desc=Optional texture to manually mask the left/right side for non-synmmetrical faces
   group=FacialExpressionBlend
   parameterName=Left/Right Mask
   samplerType=SAMPLERTYPE_Masks
   sortPriority=4
   texture=&T_BlackMask
  }
  n11 = ComponentMask(None=LocalPosition() {includedOffsets=ExcludeOffsets, localOrigin=InstancePreSkinning}.XYZ) {r=true}
  n12 = ScalarParameter() {defaultValue=1, group=FacialExpressionBlend, parameterName=Left/Right Blend Scale, sliderMax=10, sortPriority=6}
  n13 = Add(A=Multiply(A=VertexInterpolator(VS=@n11).PS, B=Divide(B=@n12))) {constB=0.5}
  n14 = StaticSwitchParameter(
   True=ChannelMaskParameter(None=@n10.RGB) {group=FacialExpressionBlend, parameterName=Left/Right Mask Channel, sortPriority=5}.R
   False=MaterialFunctionCall!(In=Saturate(None=@n13)) {materialFunction=/Game/Effects/Fort_Effects/Materials/_Functions/MF_Smoothstep_01}.Result
  ) {group=FacialExpressionBlend, parameterName=Use Left/Right Texture Mask, sortPriority=3}
  NamedRerouteDeclaration(None=@n14) {name=LeftSideMask, nodeColor=<rgba 0.5469 1 0 1>}
  n15 = TextureSample(UVs=NamedRerouteUsage(), Tex=NamedRerouteUsage()) {samplerType=SAMPLERTYPE_Masks, texture=&T_BlackMask}
  NamedRerouteDeclaration(None=@n15.R) {name=Region_Lips, nodeColor=<rgba 1 0.2812 0 1>}
  NamedRerouteDeclaration(None=@n15.G) {name=Region_BrowLateral, nodeColor=<rgba 0 0.0391 1 1>}
  NamedRerouteDeclaration(None=@n15.B) {name=Region_NoseWrinkle, nodeColor=<rgba 0.3594 1 0 1>}
  NamedRerouteDeclaration(None=@n15.A) {name=Region_CheekRaise, nodeColor=<rgba 1 0 0.6797 1>}
  n16 = TextureCoordinate()
  n17 = ComponentMask(None=@n16) {g=true}
  NamedRerouteDeclaration(None=AppendVector(A=ComponentMask(None=@n16) {r=true}, B=Multiply(A=@n17) {constB=0.5})) {name=UVs_TextureA, nodeColor=<rgba 0.0312 0 1 1>}
  n18 = Multiply(A=ComponentMask(None=@n16) {g=true}) {constB=0.5}
  NamedRerouteDeclaration(None=AppendVector(A=ComponentMask(None=@n16) {r=true}, B=Add(A=@n18) {constB=0.5})) {name=UVs_TextureB, nodeColor=<rgba 0.2891 1 0 1>}
  n19 = Multiply(
   A=ScalarParameter() {group=FacialExpressionBlend_AnimationCurves, parameterName=WM_Lips_L, sliderMax=1}
   B=Multiply(A=NamedRerouteUsage(), B=NamedRerouteUsage())
  )
  n20 = Multiply(
   A=ScalarParameter() {group=FacialExpressionBlend_AnimationCurves, parameterName=WM_Lips_R, sliderMax=1}
   B=Multiply(A=NamedRerouteUsage(), B=NamedRerouteUsage())
  )
  n21 = Multiply(
   A=ScalarParameter() {group=FacialExpressionBlend_AnimationCurves, parameterName=WM_BrowLateral_L, sliderMax=1}
   B=Multiply(A=NamedRerouteUsage(), B=NamedRerouteUsage())
  )
  n22 = Multiply(
   A=ScalarParameter() {group=FacialExpressionBlend_AnimationCurves, parameterName=WM_BrowLateral_R, sliderMax=1}
   B=Multiply(A=NamedRerouteUsage(), B=NamedRerouteUsage())
  )
  n23 = Max(
   A=Multiply(A=Add(A=@n19, B=@n20), B=ScalarParameter() {defaultValue=1, group=FacialExpressionBlend, parameterName=Scale_Lips, sliderMax=5})
   B=Multiply(A=Add(A=@n21, B=@n22), B=ScalarParameter() {defaultValue=1, group=FacialExpressionBlend, parameterName=Scale_BrowLateral, sliderMax=5})
  )
  n24 = ScalarParameter() {group=FacialExpressionBlend_AnimationCurves, parameterName=WM_NoseWrinkle_L, sliderMax=1}
  n25 = Multiply(A=NamedRerouteUsage(), B=NamedRerouteUsage())
  n26 = ScalarParameter() {group=FacialExpressionBlend_AnimationCurves, parameterName=WM_NoseWrinkle_R, sliderMax=1}
  n27 = Multiply(A=NamedRerouteUsage(), B=NamedRerouteUsage())
  n28 = Multiply(
   A=Add(A=Multiply(A=@n24, B=@n25), B=Multiply(A=@n26, B=@n27))
   B=ScalarParameter() {defaultValue=1, group=FacialExpressionBlend, parameterName=Scale_NoseWrinkle, sliderMax=5}
  )
  n29 = NamedRerouteUsage()
  n30 = NamedRerouteUsage()
  n31 = NamedRerouteUsage()
  n32 = NamedRerouteUsage()
  n33 = Add(
   A=Multiply(A=ScalarParameter() {group=FacialExpressionBlend_AnimationCurves, parameterName=WM_CheekRaise_L, sliderMax=1}, B=Multiply(A=@n29, B=@n30))
   B=Multiply(A=ScalarParameter() {group=FacialExpressionBlend_AnimationCurves, parameterName=WM_CheekRaise_R, sliderMax=1}, B=Multiply(A=@n31, B=@n32))
  )
  n34 = ScalarParameter() {defaultValue=1, group=FacialExpressionBlend, parameterName=Scale_CheekRaise, sliderMax=5}
  NamedRerouteDeclaration(None=Max(A=Max(A=@n23, B=@n28), B=Multiply(A=@n33, B=@n34))) {name=Alpha_TextureB, nodeColor=<rgba 0 1 0.2266 1>}
  n35 = TextureObjectParameter() {
   group=FacialExpressionBlend
   parameterName=FacialExpression_Mask
   samplerType=SAMPLERTYPE_Masks
   sortPriority=1
   texture=&T_BlackMask
  }
  NamedRerouteDeclaration(None=@n35) {name=Texture_Mask, nodeColor=<rgba 0 0.2031 1 1>}
  n36 = GetMaterialAttributes(None=NamedRerouteDeclaration(None=FunctionInput() {inputName=Material, inputType=FunctionInput_MaterialAttributes}) {name=IN_Material, nodeColor=<rgba 1 0.0234 0 1>}) {attributeGetTypes [#1, #2]}
  NamedRerouteDeclaration(None=@n36.AmbientOcclusion) {name=IN_AO, nodeColor=<rgba 0 1 0.4609 1>}
  n37 = TextureSample(UVs=NamedRerouteUsage(), Tex=NamedRerouteUsage()) {samplerType=SAMPLERTYPE_Masks, texture=/Game/Global/Textures/Default/Blanks/T_White}
  n38 = NamedRerouteUsage()
  n39 = NamedRerouteUsage()
  n40 = LinearInterpolate(
   A=NamedRerouteUsage()
   B=Min(
    A=NamedRerouteUsage()
    B=TextureSample(UVs=@n38, Tex=@n39) {samplerType=SAMPLERTYPE_Masks, texture=&T_White}.RGB
   )
   Alpha=NamedRerouteUsage()
  )
  NamedRerouteDeclaration(None=LinearInterpolate(A=@n40, B=Min(A=@n40, B=@n37.RGB), Alpha=NamedRerouteUsage())) {name=OUT_AO, nodeColor=<rgba 1 0.1406 0 1>}
  NamedRerouteDeclaration(None=@n36.Normal) {name=IN_Normal, nodeColor=<rgba 0 0.2969 1 1>}
  n41 = TextureObjectParameter() {
   group=FacialExpressionBlend
   parameterName=FacialExpression_Normals
   samplerType=SAMPLERTYPE_Normal
   sortPriority=2
   texture=&FlatNormalBlue
  }
  NamedRerouteDeclaration(None=@n41) {name=Texture_Normals, nodeColor=<rgba 0.5234 1 0 1>}
  n42 = TextureObjectParameter() {
   group=FacialExpressionBlend
   parameterName=FacialExpression_AO
   samplerType=SAMPLERTYPE_Masks
   sortPriority=2
   texture=&T_White
  }
  NamedRerouteDeclaration(None=@n42) {name=Texture_AO, nodeColor=<rgba 1 0 0.8437 1>}
  NamedRerouteDeclaration(None=OneMinus(None=@n14)) {name=RightSideMask, nodeColor=<rgba 1 0 0.8672 1>}
  n43 = NamedRerouteUsage()
  n44 = NamedRerouteUsage()
  n45 = LinearInterpolate(
   A=NamedRerouteUsage()
   B=MaterialFunctionCall(
    BaseNormal=NamedRerouteUsage()
    AdditionalNormal=TextureSample(UVs=@n43, Tex=@n44) {samplerType=SAMPLERTYPE_Normal, texture=&FlatNormalBlue}.RGB
   ) {materialFunction=&BlendAngleCorrectedNormals}.Result
   Alpha=NamedRerouteUsage()
  )
  n46 = MaterialFunctionCall(
   BaseNormal=@n45
   AdditionalNormal=TextureSample(UVs=NamedRerouteUsage(), Tex=NamedRerouteUsage()) {samplerType=SAMPLERTYPE_Normal, texture=&FlatNormalBlue}.RGB
  ) {materialFunction=&BlendAngleCorrectedNormals}
  n47 = NamedRerouteUsage()
  NamedRerouteDeclaration(None=Normalize(VectorInput=LinearInterpolate(A=@n45, B=@n46.Result, Alpha=@n47))) {name=OUT_Normal, nodeColor=<rgba 0 1 0.4375 1>}
  n48 = TextureSample(UVs=NamedRerouteUsage(), Tex=NamedRerouteUsage()) {samplerType=SAMPLERTYPE_Masks, texture=&T_BlackMask}
  NamedRerouteDeclaration(None=@n48.R) {name=Region_Smile, nodeColor=<rgba 0 0.4375 1 1>}
  n49 = Multiply(
   A=ScalarParameter() {group=FacialExpressionBlend_AnimationCurves, parameterName=WM_Smile_L, sliderMax=1}
   B=Multiply(A=NamedRerouteUsage(), B=NamedRerouteUsage())
  )
  n50 = Multiply(
   A=ScalarParameter() {group=FacialExpressionBlend_AnimationCurves, parameterName=WM_Smile_R, sliderMax=1}
   B=Multiply(A=NamedRerouteUsage(), B=NamedRerouteUsage())
  )
  n51 = Multiply(
   A=ScalarParameter() {group=FacialExpressionBlend_AnimationCurves, parameterName=WM_BrowRaiseInner_L, sliderMax=1}
   B=Multiply(A=NamedRerouteUsage(), B=NamedRerouteUsage())
  )
  n52 = Multiply(
   A=ScalarParameter() {group=FacialExpressionBlend_AnimationCurves, parameterName=WM_BrowRaiseInner_R, sliderMax=1}
   B=Multiply(A=NamedRerouteUsage(), B=NamedRerouteUsage())
  )
  n53 = Max(
   A=Multiply(A=Add(A=@n49, B=@n50), B=ScalarParameter() {defaultValue=1, group=FacialExpressionBlend, parameterName=Scale_Smile, sliderMax=5})
   B=Multiply(A=Add(A=@n51, B=@n52), B=ScalarParameter() {defaultValue=1, group=FacialExpressionBlend, parameterName=Scale_BrowRaiseInner, sliderMax=5})
  )
  n54 = ScalarParameter() {group=FacialExpressionBlend_AnimationCurves, parameterName=WM_BrowRaiseOutter_L, sliderMax=1}
  n55 = Multiply(A=NamedRerouteUsage(), B=NamedRerouteUsage())
  n56 = ScalarParameter() {group=FacialExpressionBlend_AnimationCurves, parameterName=WM_BrowRaiseOutter_R, sliderMax=1}
  n57 = Multiply(A=NamedRerouteUsage(), B=NamedRerouteUsage())
  n58 = Multiply(
   A=Add(A=Multiply(A=@n54, B=@n55), B=Multiply(A=@n56, B=@n57))
   B=ScalarParameter() {defaultValue=1, group=FacialExpressionBlend, parameterName=Scale_BrowRaiseOutter, sliderMax=5}
  )
  n59 = NamedRerouteUsage()
  n60 = NamedRerouteUsage()
  n61 = NamedRerouteUsage()
  n62 = NamedRerouteUsage()
  n63 = Add(
   A=Multiply(A=ScalarParameter() {group=FacialExpressionBlend_AnimationCurves, parameterName=WM_MouthStretch_L, sliderMax=1}, B=Multiply(A=@n59, B=@n60))
   B=Multiply(A=ScalarParameter() {group=FacialExpressionBlend_AnimationCurves, parameterName=WM_MouthStretch_R, sliderMax=1}, B=Multiply(A=@n61, B=@n62))
  )
  n64 = ScalarParameter() {defaultValue=1, group=FacialExpressionBlend, parameterName=Scale_MouthStretch, sliderMax=5}
  NamedRerouteDeclaration(None=Max(A=Max(A=@n53, B=@n58), B=Multiply(A=@n63, B=@n64))) {name=Alpha_TextureA, nodeColor=<rgba 0.125 1 0 1>}
  NamedRerouteDeclaration(None=@n48.G) {name=Region_BrowRaiseInner, nodeColor=<rgba 0.7578 1 0 1>}
  NamedRerouteDeclaration(None=@n48.B) {name=Region_BrowRaiseOutter, nodeColor=<rgba 0.9219 0 1 1>}
  NamedRerouteDeclaration(None=@n48.A) {name=Region_MouthStretch, nodeColor=<rgba 0 1 0.6016 1>}
 ]}
